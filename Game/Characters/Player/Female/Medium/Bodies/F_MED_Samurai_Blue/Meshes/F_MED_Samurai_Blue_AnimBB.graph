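# Animation Blueprint for the female Samurai Blue body, adding secondary motion and corrective posing to its armor and shoulders.
# The AnimGraph copies the parent mesh pose, then applies constraints and additive bone offsets weighted by thigh and upper-arm rotations measured each update.
# AnimDynamics springs swing flaps and spine cloth under skydive-dependent force, with sphere limits on front and back flaps.
# Backbling detection hides a material element and sets back-flap simulation weight for capes.
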AnimBlueprint /Game/Characters/Player/Female/Medium/Bodies/F_MED_Samurai_Blue/Meshes/F_MED_Samurai_Blue_AnimBB
target_skeleton=&/F_MED_Samurai_Blue_Skeleton
graphs [
 AnimGraph {nodes [
   n1 = Get thigh_shape_fn_l() -> (thigh_shape_fn_l:double)
   n2 = "Constraint - dyn_flap_armor_fr_01 (RR)"!(
    ComponentPose=Local To Component(LocalPose=Copy Pose From Mesh() {bUseAttachedParent=true})
    Alpha=@n1
    ConstraintWeights_0=1
    ConstraintWeights_1:float=0
   ) -> (Pose) {
    alphaBoolBlend {blendOption=Linear}
    boneToModify {boneName=dyn_flap_armor_fr_01}
    constraintSetup [{transformType=Rotation, targetBone {boneName=thigh_l}}, {transformType=Rotation, targetBone {boneName=thigh_r}}]
    constraintWeights [1,1]
   }
   n3 = Get thigh_shape_fn_r() -> (thigh_shape_fn_r:double)
   n4 = "Constraint - dyn_flap_armor_fr_01 (RR)"(ComponentPose=@n2, Alpha=@n3, ConstraintWeights_0=0, ConstraintWeights_1=1) {
    alphaBoolBlend {blendOption=Linear}
    boneToModify {boneName=dyn_flap_armor_fr_01}
    constraintSetup [{transformType=Rotation, targetBone {boneName=thigh_l}}, {transformType=Rotation, targetBone {boneName=thigh_r}}]
    constraintWeights [1,1]
   }
   n5 = "Transform (Modify) Bone - Bone: dyn_flap_armor_fr_01"(ComponentPose=@n4, Alpha=@n1, Translation="0,0,0", Rotation="-10,20,40", Scale="1,1,1") {
    rotationMode=BMM_Additive
    rotationSpace=BCS_BoneSpace
    scaleMode=BMM_Additive
    scaleSpace=BCS_BoneSpace
    translationMode=BMM_Additive
    translationSpace=BCS_BoneSpace
    alphaBoolBlend {blendOption=Linear}
    boneToModify {boneName=dyn_flap_armor_fr_01}
    rotation=<rot -10 20 40>
   }
   n6 = "Transform (Modify) Bone - Bone: dyn_flap_armor_fr_01"(ComponentPose=@n5, Alpha=@n3, Translation="0,0,0", Rotation="-10,-20,-40", Scale="1,1,1") {
    rotationMode=BMM_Additive
    rotationSpace=BCS_BoneSpace
    scaleMode=BMM_Additive
    scaleSpace=BCS_BoneSpace
    translationMode=BMM_Additive
    translationSpace=BCS_BoneSpace
    alphaBoolBlend {blendOption=Linear}
    boneToModify {boneName=dyn_flap_armor_fr_01}
    rotation=<rot -10 -20 -40>
   }
   n7 = "Constraint - dyn_flap_armor_l_01 (PP)"!(ComponentPose=@n6, Alpha=1, ConstraintWeights_0=0, ConstraintWeights_1:float=1) -> (Pose) {
    alphaBoolBlend {blendOption=Linear}
    boneToModify {boneName=dyn_flap_armor_l_01}
    constraintSetup [{targetBone {boneName=thigh_l}}, {targetBone {boneName=pelvis}}]
    constraintWeights [1,1]
   }
   n8 = "Constraint - dyn_flap_armor_r_01 (PP)"!(ComponentPose=@n7, Alpha=1, ConstraintWeights_0=0, ConstraintWeights_1:float=1) -> (Pose) {
    alphaBoolBlend {blendOption=Linear}
    boneToModify {boneName=dyn_flap_armor_r_01}
    constraintSetup [{targetBone {boneName=thigh_r}}, {targetBone {boneName=pelvis}}]
    constraintWeights [1,1]
   }
   n9 = Get thigh_shape_lt_l() -> (thigh_shape_lt_l:double)
   n10 = "Constraint - dyn_flap_armor_l_01 (PP)"(ComponentPose=@n8, Alpha=@n9, ConstraintWeights_0=1, ConstraintWeights_1=0) {
    alphaBoolBlend {blendOption=Linear}
    boneToModify {boneName=dyn_flap_armor_l_01}
    constraintSetup [{targetBone {boneName=thigh_l}}, {targetBone {boneName=pelvis}}]
    constraintWeights [1,1]
   }
   n11 = Get thigh_shape_rt_r() -> (thigh_shape_rt_r:double)
   n12 = "Constraint - dyn_flap_armor_r_01 (PP)"(ComponentPose=@n10, Alpha=@n11, ConstraintWeights_0=1, ConstraintWeights_1=0) {
    alphaBoolBlend {blendOption=Linear}
    boneToModify {boneName=dyn_flap_armor_r_01}
    constraintSetup [{targetBone {boneName=thigh_r}}, {targetBone {boneName=pelvis}}]
    constraintWeights [1,1]
   }
   n13 = "Transform (Modify) Bone - Bone: dyn_flap_armor_l_01"(ComponentPose=@n12, Alpha=@n9, Translation="0,0,3", Rotation="20,0,-20", Scale="1,1,1") {
    rotationMode=BMM_Additive
    rotationSpace=BCS_BoneSpace
    scaleMode=BMM_Additive
    scaleSpace=BCS_BoneSpace
    translationMode=BMM_Additive
    translationSpace=BCS_BoneSpace
    alphaBoolBlend {blendOption=Linear}
    boneToModify {boneName=dyn_flap_armor_l_01}
    rotation {pitch=20, roll=-20}
    translation {z=3}
   }
   n14 = "Transform (Modify) Bone - Bone: dyn_flap_armor_r_01"(ComponentPose=@n13, Alpha=@n11, Translation="0,0,-3", Rotation="20,0,-20", Scale="1,1,1") {
    rotationMode=BMM_Additive
    rotationSpace=BCS_BoneSpace
    scaleMode=BMM_Additive
    scaleSpace=BCS_BoneSpace
    translationMode=BMM_Additive
    translationSpace=BCS_BoneSpace
    alphaBoolBlend {blendOption=Linear}
    boneToModify {boneName=dyn_flap_armor_r_01}
    rotation {pitch=20, roll=-20}
    translation {z=-3}
   }
   n15 = SelectVector(A=Get skydiving vector force() -> (skydiving vector force:Vector), B="0, 0, 0", bPickA=Get bIsSkyDiving())
   n16 = Anim Dynamics - Bone: dyn_flap_armor_l_01!(ComponentPose=@n14, ExternalForce=@n15) -> (Pose) {
    angularDampingOverride=0.99
    angularSpringConstant=250
    bAngularSpring=true
    bGravityOverrideInSimSpace=true
    bOverrideAngularDamping=true
    gravityScale=0.9
    numSolverIterationsPostUpdate=4
    numSolverIterationsPreUpdate=16
    alphaBoolBlend {blendOption=Linear}
    boundBone {boneName=dyn_flap_armor_l_01}
    physicsBodyDefinitions [{
      sphereCollisionRadius=0
      boundBone {boneName=dyn_flap_armor_l_01}
      boxExtents=<8 15 15>
      constraintSetup {angularLimitsMax=<0 0 0>, angularLimitsMin=<0 -5 0>, angularTarget=<0 0 0>}
      localJointOffset=<3 0 -1>
     }]
   }
   n17 = Anim Dynamics - Bone: dyn_flap_armor_l_02!(ComponentPose=@n16, ExternalForce=@n15) -> (Pose) {
    angularDampingOverride=0.99
    angularSpringConstant=250
    bAngularSpring=true
    bGravityOverrideInSimSpace=true
    bOverrideAngularDamping=true
    gravityScale=0.9
    numSolverIterationsPostUpdate=4
    numSolverIterationsPreUpdate=16
    alphaBoolBlend {blendOption=Linear}
    boundBone {boneName=dyn_flap_armor_l_02}
    physicsBodyDefinitions [{
      sphereCollisionRadius=0
      boundBone {boneName=dyn_flap_armor_l_02}
      boxExtents=<15 30 30>
      constraintSetup {angularLimitsMax=<0 0 0>, angularLimitsMin=<0 -15 0>, angularTarget=<0 0 0>}
      localJointOffset=<3 0 -1>
     }]
   }
   n18 = Anim Dynamics - Bone: dyn_flap_armor_l_03!(ComponentPose=@n17, ExternalForce=@n15) -> (Pose) {
    angularDampingOverride=0.99
    angularSpringConstant=250
    bAngularSpring=true
    bGravityOverrideInSimSpace=true
    bOverrideAngularDamping=true
    gravityScale=0.9
    numSolverIterationsPostUpdate=4
    numSolverIterationsPreUpdate=16
    alphaBoolBlend {blendOption=Linear}
    boundBone {boneName=dyn_flap_armor_l_03}
    physicsBodyDefinitions [{
      sphereCollisionRadius=0
      boundBone {boneName=dyn_flap_armor_l_03}
      boxExtents=<8 15 15>
      constraintSetup {angularLimitsMax=<0 0 0>, angularLimitsMin=<0 -30 0>, angularTarget=<0 0 0>}
      localJointOffset=<3 0 -1>
     }]
   }
   n19 = Anim Dynamics - Bone: dyn_flap_armor_l_04!(ComponentPose=@n18, ExternalForce=@n15) -> (Pose) {
    angularDampingOverride=0.99
    angularSpringConstant=250
    bAngularSpring=true
    bGravityOverrideInSimSpace=true
    bOverrideAngularDamping=true
    gravityScale=0.9
    numSolverIterationsPostUpdate=4
    numSolverIterationsPreUpdate=16
    alphaBoolBlend {blendOption=Linear}
    boundBone {boneName=dyn_flap_armor_l_04}
    physicsBodyDefinitions [{
      sphereCollisionRadius=0
      boundBone {boneName=dyn_flap_armor_l_04}
      boxExtents=<8 15 15>
      constraintSetup {angularLimitsMax=<0 0 0>, angularLimitsMin=<0 -30 0>, angularTarget=<0 0 0>}
      localJointOffset=<3 0 -1>
     }]
   }
   n20 = Anim Dynamics - Bone: dyn_flap_armor_r_01!(ComponentPose=@n19, ExternalForce=@n15) -> (Pose) {
    angularDampingOverride=0.99
    angularSpringConstant=250
    bAngularSpring=true
    bGravityOverrideInSimSpace=true
    bOverrideAngularDamping=true
    gravityScale=0.9
    numSolverIterationsPostUpdate=4
    numSolverIterationsPreUpdate=16
    alphaBoolBlend {blendOption=Linear}
    boundBone {boneName=dyn_flap_armor_r_01}
    physicsBodyDefinitions [{
      sphereCollisionRadius=0
      boundBone {boneName=dyn_flap_armor_r_01}
      boxExtents=<8 15 15>
      constraintSetup {angularLimitsMax=<0 0 0>, angularLimitsMin=<0 -5 0>, angularTarget=<0 0 0>}
      localJointOffset=<-3 0 1>
     }]
   }
   n21 = Anim Dynamics - Bone: dyn_flap_armor_r_02!(ComponentPose=@n20, ExternalForce=@n15) -> (Pose) {
    angularDampingOverride=0.99
    angularSpringConstant=250
    bAngularSpring=true
    bGravityOverrideInSimSpace=true
    bOverrideAngularDamping=true
    gravityScale=0.9
    numSolverIterationsPostUpdate=4
    numSolverIterationsPreUpdate=16
    alphaBoolBlend {blendOption=Linear}
    boundBone {boneName=dyn_flap_armor_r_02}
    physicsBodyDefinitions [{
      sphereCollisionRadius=0
      boundBone {boneName=dyn_flap_armor_r_02}
      boxExtents=<15 30 30>
      constraintSetup {angularLimitsMax=<0 0 0>, angularLimitsMin=<0 -15 0>, angularTarget=<0 0 0>}
      localJointOffset=<-3 0 1>
     }]
   }
   n22 = Anim Dynamics - Bone: dyn_flap_armor_r_03!(ComponentPose=@n21, ExternalForce=@n15) -> (Pose) {
    angularDampingOverride=0.99
    angularSpringConstant=250
    bAngularSpring=true
    bGravityOverrideInSimSpace=true
    bOverrideAngularDamping=true
    gravityScale=0.9
    numSolverIterationsPostUpdate=4
    numSolverIterationsPreUpdate=16
    alphaBoolBlend {blendOption=Linear}
    boundBone {boneName=dyn_flap_armor_r_03}
    physicsBodyDefinitions [{
      sphereCollisionRadius=0
      boundBone {boneName=dyn_flap_armor_r_03}
      boxExtents=<8 15 15>
      constraintSetup {angularLimitsMax=<0 0 0>, angularLimitsMin=<0 -30 0>, angularTarget=<0 0 0>}
      localJointOffset=<-3 0 1>
     }]
   }
   n23 = Anim Dynamics - Bone: dyn_flap_armor_r_04!(ComponentPose=@n22, ExternalForce=@n15) -> (Pose) {
    angularDampingOverride=0.99
    angularSpringConstant=250
    bAngularSpring=true
    bGravityOverrideInSimSpace=true
    bOverrideAngularDamping=true
    gravityScale=0.9
    numSolverIterationsPostUpdate=4
    numSolverIterationsPreUpdate=16
    alphaBoolBlend {blendOption=Linear}
    boundBone {boneName=dyn_flap_armor_r_04}
    physicsBodyDefinitions [{
      sphereCollisionRadius=0
      boundBone {boneName=dyn_flap_armor_r_04}
      boxExtents=<8 15 15>
      constraintSetup {angularLimitsMax=<0 0 0>, angularLimitsMin=<0 -30 0>, angularTarget=<0 0 0>}
      localJointOffset=<-3 0 1>
     }]
   }
   n24 = float - float(B="To Float (Boolean)"(InBool=Get isWearingCape() -> (isWearingCape:bool))) {is=title K2Node_CallFunction}
   n25 = Anim Dynamics - Bone: dyn_flap_armor_bk_01(ComponentPose=@n23, Alpha=@n24, ExternalForce=@n15) {
    angularDampingOverride=0.999
    angularSpringConstant=250
    bAngularSpring=true
    bGravityOverrideInSimSpace=true
    bOverrideAngularDamping=true
    bUsePlanarLimit=false
    bUseSphericalLimits=true
    gravityScale=0.5
    alphaBoolBlend {blendOption=Linear}
    boundBone {boneName=dyn_flap_armor_bk_01}
    physicsBodyDefinitions [{
      collisionType=CustomSphere
      sphereCollisionRadius=6
      boundBone {boneName=dyn_flap_armor_bk_01}
      boxExtents=<15 40 40>
      constraintSetup {angularLimitsMax=<0 30 0>, angularLimitsMin=<0 -5 0>, angularTarget=<0 0 0>}
      localJointOffset=<15 0 -2>
     }]
    sphericalLimits [
     {limitRadius=10, drivingBone {boneName=thigh_l}, sphereLocalOffset=<-5 0 0>}
     {limitRadius=10, drivingBone {boneName=thigh_r}, sphereLocalOffset=<5 0 0>}
     {limitRadius=5, drivingBone {boneName=pelvis}, sphereLocalOffset=<-2 0 0>}
    ]
   }
   n26 = Anim Dynamics - Bone: dyn_flap_armor_bk_02(ComponentPose=@n25, Alpha=@n24, ExternalForce=@n15) {
    angularDampingOverride=0.999
    angularSpringConstant=250
    bAngularSpring=true
    bGravityOverrideInSimSpace=true
    bOverrideAngularDamping=true
    bUseSphericalLimits=true
    gravityScale=0.5
    alphaBoolBlend {blendOption=Linear}
    boundBone {boneName=dyn_flap_armor_bk_02}
    physicsBodyDefinitions [{
      collisionType=CustomSphere
      sphereCollisionRadius=6
      boundBone {boneName=dyn_flap_armor_bk_02}
      boxExtents=<15 40 40>
      constraintSetup {angularLimitsMax=<0 30 0>, angularLimitsMin=<0 -5 0>, angularTarget=<0 0 0>}
      localJointOffset=<15 0 -2>
     }]
    sphericalLimits [
     {limitRadius=10, drivingBone {boneName=thigh_l}, sphereLocalOffset=<-17 0 0>}
     {limitRadius=10, drivingBone {boneName=thigh_r}, sphereLocalOffset=<17 0 0>}
    ]
   }
   n27 = Anim Dynamics - Bone: dyn_flap_armor_bk_03(ComponentPose=@n26, Alpha=@n24, ExternalForce=@n15) {
    angularBiasOverride=0.8
    angularDampingOverride=0.999
    angularSpringConstant=250
    bAngularSpring=true
    bGravityOverrideInSimSpace=true
    bOverrideAngularBias=true
    bOverrideAngularDamping=true
    bUseSphericalLimits=true
    gravityScale=0.5
    alphaBoolBlend {blendOption=Linear}
    boundBone {boneName=dyn_flap_armor_bk_03}
    physicsBodyDefinitions [{
      collisionType=CustomSphere
      sphereCollisionRadius=6
      boundBone {boneName=dyn_flap_armor_bk_03}
      boxExtents=<15 40 40>
      constraintSetup {angularLimitsMax=<0 30 0>, angularLimitsMin=<0 -5 0>, angularTarget=<0 0 0>}
      localJointOffset=<15 0 -2>
     }]
    sphericalLimits [
     {limitRadius=10, drivingBone {boneName=thigh_l}, sphereLocalOffset=<-25 0 0>}
     {limitRadius=10, drivingBone {boneName=thigh_r}, sphereLocalOffset=<25 0 0>}
    ]
   }
   n28 = Anim Dynamics - Bone: dyn_flap_armor_fr_01(ComponentPose=@n27, Alpha=1, ExternalForce=@n15) {
    angularDampingOverride=0.999
    angularSpringConstant=250
    bAngularSpring=true
    bGravityOverrideInSimSpace=true
    bOverrideAngularDamping=true
    bUsePlanarLimit=false
    bUseSphericalLimits=true
    gravityScale=0.5
    alphaBoolBlend {blendOption=Linear}
    boundBone {boneName=dyn_flap_armor_fr_01}
    physicsBodyDefinitions [{
      collisionType=CustomSphere
      sphereCollisionRadius=6
      boundBone {boneName=dyn_flap_armor_fr_01}
      boxExtents=<15 40 40>
      constraintSetup {angularLimitsMax=<0 5 0>, angularLimitsMin=<0 -30 0>, angularTarget=<0 0 0>}
      localJointOffset=<15 0 -2>
     }]
    sphericalLimits [
     {limitRadius=10, drivingBone {boneName=thigh_l}, sphereLocalOffset=<-5 0 0>}
     {limitRadius=10, drivingBone {boneName=thigh_r}, sphereLocalOffset=<5 0 0>}
     {limitRadius=5, drivingBone {boneName=pelvis}, sphereLocalOffset=<-2 0 0>}
    ]
   }
   n29 = Anim Dynamics - Bone: dyn_flap_armor_fr_02(ComponentPose=@n28, Alpha=1, ExternalForce=@n15) {
    angularDampingOverride=0.999
    angularSpringConstant=250
    bAngularSpring=true
    bGravityOverrideInSimSpace=true
    bOverrideAngularDamping=true
    bUseSphericalLimits=true
    gravityScale=0.5
    alphaBoolBlend {blendOption=Linear}
    boundBone {boneName=dyn_flap_armor_fr_02}
    physicsBodyDefinitions [{
      collisionType=CustomSphere
      sphereCollisionRadius=6
      boundBone {boneName=dyn_flap_armor_fr_02}
      boxExtents=<15 40 40>
      constraintSetup {angularLimitsMax=<0 5 0>, angularLimitsMin=<0 -30 0>, angularTarget=<0 0 0>}
      localJointOffset=<15 0 -2>
     }]
    sphericalLimits [
     {limitRadius=10, drivingBone {boneName=thigh_l}, sphereLocalOffset=<-17 0 0>}
     {limitRadius=10, drivingBone {boneName=thigh_r}, sphereLocalOffset=<17 0 0>}
    ]
   }
   n30 = Anim Dynamics - Bone: dyn_flap_armor_fr_03(ComponentPose=@n29, Alpha=1, ExternalForce=@n15) {
    angularBiasOverride=0.8
    angularDampingOverride=0.999
    angularSpringConstant=250
    bAngularSpring=true
    bGravityOverrideInSimSpace=true
    bOverrideAngularBias=true
    bOverrideAngularDamping=true
    bUseSphericalLimits=true
    gravityScale=0.5
    alphaBoolBlend {blendOption=Linear}
    boundBone {boneName=dyn_flap_armor_fr_03}
    physicsBodyDefinitions [{
      collisionType=CustomSphere
      sphereCollisionRadius=6
      boundBone {boneName=dyn_flap_armor_fr_03}
      boxExtents=<15 40 40>
      constraintSetup {angularLimitsMax=<0 5 0>, angularLimitsMin=<0 -30 0>, angularTarget=<0 0 0>}
      localJointOffset=<15 0 -2>
     }]
    sphericalLimits [
     {limitRadius=10, drivingBone {boneName=thigh_l}, sphereLocalOffset=<-25 0 0>}
     {limitRadius=10, drivingBone {boneName=thigh_r}, sphereLocalOffset=<25 0 0>}
    ]
   }
   n31 = Anim Dynamics - Bone: dyn_spine_cloth_l!(ComponentPose=@n30, ExternalForce=@n15) -> (Pose) {
    angularDampingOverride=0.9
    angularSpringConstant=500
    bAngularSpring=true
    bGravityOverrideInSimSpace=true
    bOverrideAngularDamping=true
    numSolverIterationsPostUpdate=4
    numSolverIterationsPreUpdate=16
    alphaBoolBlend {blendOption=Linear}
    boundBone {boneName=dyn_spine_cloth_l}
    physicsBodyDefinitions [{
      sphereCollisionRadius=0
      boundBone {boneName=dyn_spine_cloth_l}
      boxExtents=<20 20 10>
      constraintSetup {angularTargetAxis=AxisZ, twistAxis=AxisZ, angularLimitsMax=<5 8 2>, angularLimitsMin=<-5 -40 -2>, angularTarget=<0 0 1>}
      localJointOffset=<0 0 -2>
     }]
   }
   n32 = Anim Dynamics - Bone: dyn_spine_cloth_r!(ComponentPose=@n31, ExternalForce=@n15) -> (Pose) {
    angularDampingOverride=0.9
    angularSpringConstant=500
    bAngularSpring=true
    bGravityOverrideInSimSpace=true
    bOverrideAngularDamping=true
    numSolverIterationsPostUpdate=4
    numSolverIterationsPreUpdate=16
    alphaBoolBlend {blendOption=Linear}
    boundBone {boneName=dyn_spine_cloth_r}
    physicsBodyDefinitions [{
      sphereCollisionRadius=0
      boundBone {boneName=dyn_spine_cloth_r}
      boxExtents=<20 20 10>
      constraintSetup {angularTargetAxis=AxisZ, twistAxis=AxisZ, angularLimitsMax=<5 8 2>, angularLimitsMin=<-5 -40 -2>, angularTarget=<0 0 1>}
      localJointOffset=<0 0 2>
     }]
   }
   n33 = Get Shoulder_rot_up_l() -> (Shoulder_rot_up_l:double)
   n34 = "Constraint - upperarmUp_l (PP)"!(ComponentPose=@n32, Alpha=@n33, ConstraintWeights_0=0.5, ConstraintWeights_1:float=0.5) -> (Pose) {
    alphaBoolBlend {blendOption=Linear}
    boneToModify {boneName=upperarmUp_l}
    constraintSetup [{targetBone {boneName=upperarm_twist_01_l}}, {targetBone {boneName=spine_05}}]
    constraintWeights [1,1]
   }
   n35 = Get Shoulder_rot_up_r() -> (Shoulder_rot_up_r:double)
   n36 = "Constraint - upperarmUp_r (PP)"!(ComponentPose=@n34, Alpha=@n35, ConstraintWeights_0=0.5, ConstraintWeights_1:float=0.5) -> (Pose) {
    alphaBoolBlend {blendOption=Linear}
    boneToModify {boneName=upperarmUp_r}
    constraintSetup [{targetBone {boneName=upperarm_twist_01_r}}, {targetBone {boneName=spine_05}}]
    constraintWeights [1,1]
   }
   n37 = "Transform (Modify) Bone - Bone: upperarmUp_l"(ComponentPose=@n36, Alpha=@n33, Translation="-8,0,8", Rotation="50,0,0", Scale="1,1,1") {
    rotationMode=BMM_Additive
    rotationSpace=BCS_BoneSpace
    scaleMode=BMM_Additive
    scaleSpace=BCS_BoneSpace
    translationMode=BMM_Additive
    translationSpace=BCS_BoneSpace
    alphaBoolBlend {blendOption=Linear}
    boneToModify {boneName=upperarmUp_l}
    rotation {pitch=50}
    translation {x=-8, z=8}
   }
   n38 = "Transform (Modify) Bone - Bone: upperarmUp_r"(ComponentPose=@n37, Alpha=@n35, Translation="8,0,-8", Rotation="50,0,0", Scale="1,1,1") {
    rotationMode=BMM_Additive
    rotationSpace=BCS_BoneSpace
    scaleMode=BMM_Additive
    scaleSpace=BCS_BoneSpace
    translationMode=BMM_Additive
    translationSpace=BCS_BoneSpace
    alphaBoolBlend {blendOption=Linear}
    boneToModify {boneName=upperarmUp_r}
    rotation {pitch=50}
    translation {x=8, z=-8}
   }
   n39 = Get Shoulder_rot_f_l() -> (Shoulder_rot_f_l:double)
   n40 = "Constraint - upperarmUp_l (PP)"(ComponentPose=@n38, Alpha=@n39, ConstraintWeights_0=0.5, ConstraintWeights_1=0.5) {
    alphaBoolBlend {blendOption=Linear}
    boneToModify {boneName=upperarmUp_l}
    constraintSetup [{targetBone {boneName=upperarm_twist_01_l}}, {targetBone {boneName=spine_05}}]
    constraintWeights [1,1]
   }
   n41 = Get Shoulder_rot_f_r() -> (Shoulder_rot_f_r:double)
   n42 = "Constraint - upperarmUp_r (PP)"(ComponentPose=@n40, Alpha=@n41, ConstraintWeights_0=0.5, ConstraintWeights_1=0.5) {
    alphaBoolBlend {blendOption=Linear}
    boneToModify {boneName=upperarmUp_r}
    constraintSetup [{targetBone {boneName=upperarm_twist_01_r}}, {targetBone {boneName=spine_05}}]
    constraintWeights [1,1]
   }
   n43 = "Transform (Modify) Bone - Bone: upperarmUp_l"(ComponentPose=@n42, Alpha=@n39, Translation="-8,0,3", Rotation="50,0,0", Scale="1,1,1") {
    rotationMode=BMM_Additive
    rotationSpace=BCS_BoneSpace
    scaleMode=BMM_Additive
    scaleSpace=BCS_BoneSpace
    translationMode=BMM_Additive
    translationSpace=BCS_BoneSpace
    alphaBoolBlend {blendOption=Linear}
    boneToModify {boneName=upperarmUp_l}
    rotation {pitch=50}
    translation {x=-8, z=3}
   }
   n44 = "Transform (Modify) Bone - Bone: upperarmUp_r"(ComponentPose=@n43, Alpha=@n41, Translation="8,0,-3", Rotation="50,0,0", Scale="1,1,1") {
    rotationMode=BMM_Additive
    rotationSpace=BCS_BoneSpace
    scaleMode=BMM_Additive
    scaleSpace=BCS_BoneSpace
    translationMode=BMM_Additive
    translationSpace=BCS_BoneSpace
    alphaBoolBlend {blendOption=Linear}
    boneToModify {boneName=upperarmUp_r}
    rotation {pitch=50}
    translation {x=8, z=-3}
   }
   n45 = "Transform (Modify) Bone - Bone: upperarmFr_l"(ComponentPose=@n44, Alpha=Get Shoulder_rot_f_l(), Translation="-8,2,0", Rotation="0,0,0", Scale="1,1,1") {
    rotationMode=BMM_Additive
    rotationSpace=BCS_BoneSpace
    scaleMode=BMM_Additive
    scaleSpace=BCS_BoneSpace
    translationMode=BMM_Additive
    translationSpace=BCS_BoneSpace
    alphaBoolBlend {blendOption=Linear}
    boneToModify {boneName=upperarmFr_l}
    translation=<-8 2>
   }
   n46 = "Transform (Modify) Bone - Bone: upperarmFr_r"(ComponentPose=@n45, Alpha=Get Shoulder_rot_f_r(), Translation="8,-2,0", Rotation="0,0,0", Scale="1,1,1") {
    rotationMode=BMM_Additive
    rotationSpace=BCS_BoneSpace
    scaleMode=BMM_Additive
    scaleSpace=BCS_BoneSpace
    translationMode=BMM_Additive
    translationSpace=BCS_BoneSpace
    alphaBoolBlend {blendOption=Linear}
    boneToModify {boneName=upperarmFr_r}
    translation=<8 -2>
   }
   n47 = "Transform (Modify) Bone - Bone: upperarm_twist_01_l"(ComponentPose=@n46, Alpha=Get Shoulder_rot_f_l(), Translation="0,0,0", Rotation="0,0,-20", Scale="1,1,1") {
    rotationMode=BMM_Additive
    rotationSpace=BCS_BoneSpace
    scaleMode=BMM_Additive
    scaleSpace=BCS_BoneSpace
    translationMode=BMM_Additive
    translationSpace=BCS_BoneSpace
    alphaBoolBlend {blendOption=Linear}
    boneToModify {boneName=upperarm_twist_01_l}
    rotation {roll=-20}
   }
   n48 = "Transform (Modify) Bone - Bone: upperarm_twist_01_r"(ComponentPose=@n47, Alpha=Get Shoulder_rot_f_r(), Translation="0,0,0", Rotation="0,0,20", Scale="1,1,1") {
    rotationMode=BMM_Additive
    rotationSpace=BCS_BoneSpace
    scaleMode=BMM_Additive
    scaleSpace=BCS_BoneSpace
    translationMode=BMM_Additive
    translationSpace=BCS_BoneSpace
    alphaBoolBlend {blendOption=Linear}
    boneToModify {boneName=upperarm_twist_01_r}
    rotation {roll=20}
   }
   n49 = "Transform (Modify) Bone - Bone: upperarm_twist_01_l"(ComponentPose=@n48, Alpha=Get Shoulder_rot_up_l(), Translation="0,0,0", Rotation="0,0,15", Scale="1,1,1") {
    rotationMode=BMM_Additive
    rotationSpace=BCS_BoneSpace
    scaleMode=BMM_Additive
    scaleSpace=BCS_BoneSpace
    translationMode=BMM_Additive
    translationSpace=BCS_BoneSpace
    alphaBoolBlend {blendOption=Linear}
    boneToModify {boneName=upperarm_twist_01_l}
    rotation {roll=15}
   }
   n50 = "Transform (Modify) Bone - Bone: upperarm_twist_01_r"(ComponentPose=@n49, Alpha=Get Shoulder_rot_up_r(), Translation="0,0,0", Rotation="0,0,-15", Scale="1,1,1") {
    rotationMode=BMM_Additive
    rotationSpace=BCS_BoneSpace
    scaleMode=BMM_Additive
    scaleSpace=BCS_BoneSpace
    translationMode=BMM_Additive
    translationSpace=BCS_BoneSpace
    alphaBoolBlend {blendOption=Linear}
    boneToModify {boneName=upperarm_twist_01_r}
    rotation {roll=-15}
   }
   n51 = "Transform (Modify) Bone - Bone: upperarmBack_l"(ComponentPose=@n50, Alpha=Get Shoulder_rot_f_l(), Translation="3,-3,-3", Rotation="0,0,0", Scale="1.2,1.2,1.2") {
    rotationMode=BMM_Additive
    rotationSpace=BCS_BoneSpace
    scaleMode=BMM_Additive
    scaleSpace=BCS_BoneSpace
    translationMode=BMM_Additive
    translationSpace=BCS_BoneSpace
    alphaBoolBlend {blendOption=Linear}
    boneToModify {boneName=upperarmBack_l}
    scale=<1.2 1.2 1.2>
    translation=<3 -3 -3>
   }
   n52 = "Transform (Modify) Bone - Bone: upperarmBack_r"(ComponentPose=@n51, Alpha=Get Shoulder_rot_f_r(), Translation="5,5,-5", Rotation="0,0,0", Scale="1.5,1.2,1.5") {
    rotationMode=BMM_Additive
    rotationSpace=BCS_BoneSpace
    scaleMode=BMM_Additive
    scaleSpace=BCS_BoneSpace
    translationMode=BMM_Additive
    translationSpace=BCS_BoneSpace
    alphaBoolBlend {blendOption=Linear}
    boneToModify {boneName=upperarmBack_r}
    scale=<1.5 1.2 1.5>
    translation=<5 5 -5>
   }
   n53 = "Constraint - upperarmDn_l (PP)"!(ComponentPose=@n52, Alpha=1, ConstraintWeights_0=0.15, ConstraintWeights_1:float=0.85) -> (Pose) {
    alphaBoolBlend {blendOption=Linear}
    boneToModify {boneName=upperarmDn_l}
    constraintSetup [{targetBone {boneName=upperarm_twist_01_l}}, {targetBone {boneName=clavicle_l}}]
    constraintWeights [1,1]
   }
   n54 = "Constraint - upperarmDn_r (PP)"!(ComponentPose=@n53, Alpha=1, ConstraintWeights_0=0.15, ConstraintWeights_1:float=0.85) -> (Pose) {
    alphaBoolBlend {blendOption=Linear}
    boneToModify {boneName=upperarmDn_r}
    constraintSetup [{targetBone {boneName=upperarm_twist_01_r}}, {targetBone {boneName=clavicle_r}}]
    constraintWeights [1,1]
   }
   n55 = "Transform (Modify) Bone - Bone: upperarmDn_l"(ComponentPose=@n54, Alpha=Get Shoulder_rot_dn_l() -> (Shoulder_rot_dn_l:double), Translation="-11,-3,5", Rotation="0,0,0", Scale="1,1,1") {
    rotationMode=BMM_Additive
    rotationSpace=BCS_BoneSpace
    scaleMode=BMM_Additive
    scaleSpace=BCS_BoneSpace
    translationMode=BMM_Additive
    translationSpace=BCS_BoneSpace
    alphaBoolBlend {blendOption=Linear}
    boneToModify {boneName=upperarmDn_l}
    translation=<-11 -3 5>
   }
   n56 = "Transform (Modify) Bone - Bone: upperarmDn_r"(ComponentPose=@n55, Alpha=Get Shoulder_rot_dn_r() -> (Shoulder_rot_dn_r:double), Translation="11,3,-5", Rotation="0,0,0", Scale="1,1,1") {
    rotationMode=BMM_Additive
    rotationSpace=BCS_BoneSpace
    scaleMode=BMM_Additive
    scaleSpace=BCS_BoneSpace
    translationMode=BMM_Additive
    translationSpace=BCS_BoneSpace
    alphaBoolBlend {blendOption=Linear}
    boneToModify {boneName=upperarmDn_r}
    translation=<11 3 -5>
   }
   n57 = "Transform (Modify) Bone - Bone: upperarmDn_l"(ComponentPose=@n56, Alpha=Get Shoulder_rot_f_l(), Translation="2,0,5", Rotation="0,0,0", Scale="1,1,1") {
    rotationMode=BMM_Additive
    rotationSpace=BCS_BoneSpace
    scaleMode=BMM_Additive
    scaleSpace=BCS_BoneSpace
    translationMode=BMM_Additive
    translationSpace=BCS_BoneSpace
    alphaBoolBlend {blendOption=Linear}
    boneToModify {boneName=upperarmDn_l}
    translation {x=2, z=5}
   }
   n58 = "Transform (Modify) Bone - Bone: upperarmDn_r"(ComponentPose=@n57, Alpha=Get Shoulder_rot_f_r(), Translation="0,0,-5", Rotation="0,0,0", Scale="1,1,1") {
    rotationMode=BMM_Additive
    rotationSpace=BCS_BoneSpace
    scaleMode=BMM_Additive
    scaleSpace=BCS_BoneSpace
    translationMode=BMM_Additive
    translationSpace=BCS_BoneSpace
    alphaBoolBlend {blendOption=Linear}
    boneToModify {boneName=upperarmDn_r}
    translation {z=-5}
   }
   n59 = "Transform (Modify) Bone - Bone: shoulder_armor_l"(ComponentPose=@n58, Alpha=Get Shoulder_rot_up_l(), Translation="2,-5,-4", Rotation="0,0,-75", Scale="1,1,1") {
    rotationMode=BMM_Additive
    rotationSpace=BCS_BoneSpace
    scaleMode=BMM_Additive
    scaleSpace=BCS_BoneSpace
    translationMode=BMM_Additive
    translationSpace=BCS_BoneSpace
    alphaBoolBlend {blendOption=Linear}
    boneToModify {boneName=shoulder_armor_l}
    rotation {roll=-75}
    translation=<2 -5 -4>
   }
   n60 = "Transform (Modify) Bone - Bone: shoulder_armor_l"(ComponentPose=@n59, Alpha=Get Shoulder_rot_f_l(), Translation="0,2,2", Rotation="0,0,15", Scale="1,1,1") {
    rotationMode=BMM_Additive
    rotationSpace=BCS_BoneSpace
    scaleMode=BMM_Additive
    scaleSpace=BCS_BoneSpace
    translationMode=BMM_Additive
    translationSpace=BCS_BoneSpace
    alphaBoolBlend {blendOption=Linear}
    boneToModify {boneName=shoulder_armor_l}
    rotation {roll=15}
    translation {y=2, z=2}
   }
   Output Pose(Result=Component To Local(ComponentPose=@n60)) {layerGroup=None}
  ]}
 EventGraph {nodes [
   n1 = Event OnCharacterCustomizationLoaded()
   Do Once(execute=[@n7.then, @n1.then])
   n2 = CheckForBackbling()
   > Set isWearingBackbling(isWearingBackbling=OR Boolean(A=@n2, B=@n2.bIsBackblingLong))
   > Branch(Condition=@n2)
   > Set isWearingCape(isWearingCape=@n2.bIsBackblingLong)
   Event BlueprintUpdateAnimation()
   > n3 = Sequence!(execute) -> (then_0, then_1, then_2:Exec)
   > n4 = Branch(Condition=Get isWearingBackbling())
   n5 = GetPartSkeletalMeshComponent()
   > n6 = Get Dynamic Material Instance(SkeletalMeshComponent=@n5)
   > n7 = SetScalarParameterValue(self=@n6, ParameterName=Hide Element 02, Value=1)
   Reroute Node(InputPin=@n4.else)
   n8 = Branch(execute=@n3.then_1, Condition=Get bCanPlayCustomAnimations() -> (bCanPlayCustomAnimations:bool))
   n9 = Branch(execute=[@n8.else, @n8.then], Condition=Get bIsSkyDiving())
   n10 = Branch(execute=[@n9.else, @n9.then], Condition=Get bIsSkydiveFloating())
   n11 = Get Direction Between Sockets!(
    Component:Skeletal Mesh Component Object Reference=GetOwnerSkeletalMeshComponent()
    SocketOrBoneNameFrom:Name=head
    SocketOrBoneNameTo:Name=pelvis
   ) -> (ReturnValue:Vector) {is=title K2Node_CallFunction}
   > Set skydiving vector force(skydiving vector force:Vector=vector + vector(
     A:Vector=vector * vector(A:Vector=@n11, B:Vector="75,75,75") -> (ReturnValue:Vector) {is=title K2Node_CommutativeAssociativeBinaryOperator}
     B:Vector="0, 0,2000"
    ) -> (ReturnValue:Vector) {is=title K2Node_CommutativeAssociativeBinaryOperator}) -> (Output_Get:Vector)
   n12 = Branch(execute=@n10.else, Condition=Get bIsSkydiveDiveMode())
   > Set skydiving vector force(skydiving vector force=vector + vector(A=vector * vector(A=@n11, B="180,180,200"), B="0, 0,2000"))
   Set skydiving vector force(execute=@n12.else, skydiving vector force=vector * vector(A=@n11, B="-5,-5,-5"))
   n13 = Get Skel_Mesh()
   Set thigh_shape_fn_l(
    execute=@n3.then_2
    thigh_shape_fn_l:double=GetAlphaFromReferenceTransformDelta(SkeletalMeshComponent=@n13, BoneName=thigh_l, BaseName=pelvis, InputRange=(X=0,Y=-90), TwistAxis="0,0,1")
   ) -> (Output_Get:double)
   > Set thigh_shape_fn_r(thigh_shape_fn_r:double=GetAlphaFromReferenceTransformDelta(SkeletalMeshComponent=@n13, BoneName=thigh_r, BaseName=pelvis, InputRange=(X=0,Y=-90), TwistAxis="0,0,1")) -> (Output_Get:double)
   n14 = Get Skel_Mesh()
   > Set thigh_shape_lt_l(thigh_shape_lt_l:double=GetAlphaFromReferenceTransformDelta(
     SkeletalMeshComponent=@n14
     BoneName=thigh_l
     BaseName=pelvis
     DeltaComponentType=EulerY
     InputRange=(X=60,Y=-30)
     OutputRange=(X=1,Y=0)
     TwistAxis="0,0,1"
    )) -> (Output_Get:double)
   > Set thigh_shape_rt_r(thigh_shape_rt_r:double=GetAlphaFromReferenceTransformDelta(
     SkeletalMeshComponent=@n14
     BoneName=thigh_r
     BaseName=pelvis
     DeltaComponentType=EulerY
     InputRange=(X=60,Y=-30)
     OutputRange=(X=1,Y=0)
     TwistAxis="0,0,1"
    )) -> (Output_Get:double)
   n15 = Get Shoulder_Mesh() -> (Shoulder_Mesh:Skeletal Mesh Component Object Reference)
   > Set Shoulder_rot_f_r(Shoulder_rot_f_r:double=GetAlphaFromReferenceTransformDelta(SkeletalMeshComponent=@n15, BoneName=upperarm_r, BaseName=clavicle_r, InputRange=(X=0,Y=110), TwistAxis="0,0,1")) -> (Output_Get:double)
   > Set Shoulder_rot_f_l(Shoulder_rot_f_l:double=GetAlphaFromReferenceTransformDelta(SkeletalMeshComponent=@n15, BoneName=upperarm_l, BaseName=clavicle_l, InputRange=(X=0,Y=110), TwistAxis="0,0,1")) -> (Output_Get:double)
   n16 = Get Shoulder_Mesh()
   > Set Shoulder_rot_dn_r(Shoulder_rot_dn_r:double=GetAlphaFromReferenceTransformDelta(SkeletalMeshComponent=@n16, BoneName=upperarm_r, BaseName=clavicle_r, DeltaComponentType=EulerY, InputRange=(X=0,Y=-90), TwistAxis="0,0,1")) -> (Output_Get:double)
   > Set Shoulder_rot_dn_l(Shoulder_rot_dn_l:double=GetAlphaFromReferenceTransformDelta(SkeletalMeshComponent=@n16, BoneName=upperarm_l, BaseName=clavicle_l, DeltaComponentType=EulerY, InputRange=(X=0,Y=-90), TwistAxis="0,0,1")) -> (Output_Get:double)
   n17 = Get Shoulder_Mesh()
   > Set Shoulder_rot_up_r(Shoulder_rot_up_r:double=GetAlphaFromReferenceTransformDelta(SkeletalMeshComponent=@n17, BoneName=upperarm_r, BaseName=clavicle_r, DeltaComponentType=EulerY, InputRange=(X=0,Y=60), TwistAxis="0,0,1")) -> (Output_Get:double)
   > Set Shoulder_rot_up_l(Shoulder_rot_up_l:double=GetAlphaFromReferenceTransformDelta(SkeletalMeshComponent=@n17, BoneName=upperarm_l, BaseName=clavicle_l, DeltaComponentType=EulerY, InputRange=(X=0,Y=50), TwistAxis="0,0,1")) -> (Output_Get:double)
   Event BlueprintInitializeAnimation()
   > Set Skel_Mesh(Skel_Mesh=GetOwnerSkeletalMeshComponent())
   > n18 = Cast To PlayerPawn_Athena(Object=TryGetPawnOwner())
   > n19 = Set Pawn Owner(Pawn Owner:Fort Player Pawn Athena Object Reference=@n18) -> (Output_Get:Fort Player Pawn Athena Object Reference)
   > Set Shoulder_Mesh(Shoulder_Mesh:Skeletal Mesh Component Object Reference=GetSkeletalMeshForPartType(self=@n19)) -> (Output_Get:Skeletal Mesh Component Object Reference)
   TryGetPawnOwner()
   n20 = FindChildCommponentFromAttachSocket!(
    self:Custom Character Part Anim Instance Object Reference
    Parent:Skeletal Mesh Component Object Reference=@n5
    bIncludeAllDescendants:bool=false
    AttachSocket:Name=Cape
   ) -> (ReturnValue:Skeletal Mesh Component Object Reference)
   "Not Equal (Object)"!(A:Object Reference=@n20, B:Object Reference) -> (ReturnValue:bool)
   "float > float"(
    A=Get Distance Between Two Sockets!(
     Component:Skeletal Mesh Component Object Reference=@n20
     SocketOrBoneNameA:Name=backpack_root
     SocketSpaceA:ERelativeTransformSpace Enum=RTS_World
     SocketOrBoneNameB:Name=cape_length
     SocketSpaceB:ERelativeTransformSpace Enum=RTS_World
     bRemapRange:bool=false
     InRangeMin:float=0
     InRangeMax:float=0
     OutRangeMin:float=0
     OutRangeMax:float=0
    ) -> (ReturnValue:float) {is=title K2Node_CallFunction}
    B=50
   ) {is=title K2Node_CallFunction}
   "Not Equal (Object)"(A=FindChildCommponentFromAttachSocket(Parent=@n5, bIncludeAllDescendants=false, AttachSocket=backpack_br))
  ]}
]
variables [
 isWearingBackbling {type=boolean}
 isWearingCape {type=boolean}
 Pawn Owner {type=FortPlayerPawnAthena}
 Shoulder_Mesh {type=SkeletalMeshComponent}
 Shoulder_rot_dn_l {type=number}
 Shoulder_rot_dn_r {type=number}
 Shoulder_rot_f_l {type=number}
 Shoulder_rot_f_r {type=number}
 Shoulder_rot_up_l {type=number}
 Shoulder_rot_up_r {type=number}
 Skel_Mesh {type=SkeletalMeshComponent}
 skydiving vector force {type=Vector}
 thigh_shape_fn_l {type=number}
 thigh_shape_fn_r {type=number}
 thigh_shape_lt_l {type=number}
 thigh_shape_rt_r {type=number}
]
